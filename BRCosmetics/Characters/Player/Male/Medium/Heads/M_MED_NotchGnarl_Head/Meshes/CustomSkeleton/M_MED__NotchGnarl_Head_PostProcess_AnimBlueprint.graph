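# Post-process AnimBlueprint for the NotchGnarl male head, adding secondary fur motion and shrinking head and neck bones.
# Chained spring-driven AnimDynamics nodes simulate seven fur bones, then bone-space Modify Bone nodes scale head and neck_02.
# The update event sets sim-space and dynamics-alpha variables by front-end state.
# It also flags scale-down when the local player aims a visible weapon, excluding trap, rocket launcher and GoingCommando animation types.
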AnimBlueprint /BRCosmetics/Characters/Player/Male/Medium/Heads/M_MED_NotchGnarl_Head/Meshes/CustomSkeleton/M_MED__NotchGnarl_Head_PostProcess_AnimBlueprint
target_skeleton=&/Fortnite_Base_Head_NotchGnarl_Export_Skeleton
graphs [
 AnimGraph {nodes [
   n1 = Anim Dynamics - Bone: base_headfur_m!(SimulationSpace:AnimPhysSimSpaceType Enum, ComponentPose=Local To Component(LocalPose=Input Pose()), Alpha=1, ExternalForce="0,0,0") -> (Pose) {
    angularDampingOverride=1
    angularSpringConstant=2000
    bAngularSpring=true
    bOverrideAngularDamping=true
    bUseGravityOverride=true
    gravityScale=0.35
    simulationSpace=RootRelative
    boundBone {boneName=base_headfur_m}
    componentAppliedLinearAccClamp=<2000 2000 2000>
    componentLinearAccScale=<0.2 0.2 0.2>
    componentLinearVelScale=<0.2 0.2 0.2>
    gravityOverride {z=-980}
    physicsBodyDefinitions [{
      boundBone {boneName=base_headfur_m}
      boxExtents=<30 30 30>
      constraintSetup {angularLimitsMax=<0 2 2>, angularLimitsMin=<0 -2 -2>, angularTarget=<1 0 0>}
      localJointOffset=<2 0 0>
     }]
   }
   n2 = Anim Dynamics - Bone: dyn_headfur_1_01!(SimulationSpace:AnimPhysSimSpaceType Enum, ComponentPose=@n1, Alpha=1, ExternalForce="0,0,0") -> (Pose) {
    angularDampingOverride=1
    angularSpringConstant=14000
    bAngularSpring=true
    bOverrideAngularDamping=true
    bUseGravityOverride=true
    gravityScale=0.35
    simulationSpace=RootRelative
    boundBone {boneName=dyn_headfur_1_01}
    componentAppliedLinearAccClamp=<2000 2000 2000>
    componentLinearAccScale=<1 1 1>
    componentLinearVelScale=<1 1 1>
    gravityOverride {z=-980}
    physicsBodyDefinitions [{
      boundBone {boneName=dyn_headfur_1_01}
      boxExtents=<25 30 30>
      constraintSetup {angularLimitsMax=<0 5 5>, angularLimitsMin=<0 -5 -5>, angularTarget=<1 0 0>}
      localJointOffset=<2 0 0>
     }]
   }
   n3 = Anim Dynamics - Bone: dyn_headfur_1_02!(SimulationSpace:AnimPhysSimSpaceType Enum, ComponentPose=@n2, Alpha=1, ExternalForce="0,0,0") -> (Pose) {
    angularDampingOverride=1
    angularSpringConstant=14000
    bAngularSpring=true
    bOverrideAngularDamping=true
    bUseGravityOverride=true
    gravityScale=0.35
    simulationSpace=RootRelative
    boundBone {boneName=dyn_headfur_1_02}
    componentAppliedLinearAccClamp=<2000 2000 2000>
    componentLinearAccScale=<1 1 1>
    componentLinearVelScale=<1 1 1>
    gravityOverride {z=-980}
    physicsBodyDefinitions [{
      boundBone {boneName=dyn_headfur_1_02}
      boxExtents=<25 30 30>
      constraintSetup {angularLimitsMax=<0 5 5>, angularLimitsMin=<0 -5 -5>, angularTarget=<1 0 0>}
      localJointOffset=<2 0 0>
     }]
   }
   n4 = Anim Dynamics - Bone: dyn_headfur_2_01!(SimulationSpace:AnimPhysSimSpaceType Enum, ComponentPose=@n3, Alpha=1, ExternalForce="0,0,0") -> (Pose) {
    angularDampingOverride=1
    angularSpringConstant=14000
    bAngularSpring=true
    bOverrideAngularDamping=true
    bUseGravityOverride=true
    gravityScale=0.35
    simulationSpace=RootRelative
    boundBone {boneName=dyn_headfur_2_01}
    componentAppliedLinearAccClamp=<2000 2000 2000>
    componentLinearAccScale=<1 1 1>
    componentLinearVelScale=<1 1 1>
    gravityOverride {z=-980}
    physicsBodyDefinitions [{
      boundBone {boneName=dyn_headfur_2_01}
      boxExtents=<25 30 30>
      constraintSetup {angularLimitsMax=<0 5 5>, angularLimitsMin=<0 -5 -5>, angularTarget=<1 0 0>}
      localJointOffset=<2 0 0>
     }]
   }
   n5 = Anim Dynamics - Bone: dyn_headfur_2_02!(SimulationSpace:AnimPhysSimSpaceType Enum, ComponentPose=@n4, Alpha=1, ExternalForce="0,0,0") -> (Pose) {
    angularDampingOverride=1
    angularSpringConstant=14000
    bAngularSpring=true
    bOverrideAngularDamping=true
    bUseGravityOverride=true
    gravityScale=0.35
    simulationSpace=RootRelative
    boundBone {boneName=dyn_headfur_2_02}
    componentAppliedLinearAccClamp=<2000 2000 2000>
    componentLinearAccScale=<1 1 1>
    componentLinearVelScale=<1 1 1>
    gravityOverride {z=-980}
    physicsBodyDefinitions [{
      boundBone {boneName=dyn_headfur_2_02}
      boxExtents=<25 30 30>
      constraintSetup {angularLimitsMax=<0 5 5>, angularLimitsMin=<0 -5 -5>, angularTarget=<1 0 0>}
      localJointOffset=<2 0 0>
     }]
   }
   n6 = Anim Dynamics - Bone: dyn_headfur_3_01!(SimulationSpace:AnimPhysSimSpaceType Enum, ComponentPose=@n5, Alpha=1, ExternalForce="0,0,0") -> (Pose) {
    angularDampingOverride=1
    angularSpringConstant=14000
    bAngularSpring=true
    bOverrideAngularDamping=true
    bUseGravityOverride=true
    gravityScale=0.35
    simulationSpace=RootRelative
    boundBone {boneName=dyn_headfur_3_01}
    componentAppliedLinearAccClamp=<2000 2000 2000>
    componentLinearAccScale=<1 1 1>
    componentLinearVelScale=<1 1 1>
    gravityOverride {z=-980}
    physicsBodyDefinitions [{
      boundBone {boneName=dyn_headfur_3_01}
      boxExtents=<25 30 30>
      constraintSetup {angularLimitsMax=<0 5 5>, angularLimitsMin=<0 -5 -5>, angularTarget=<1 0 0>}
      localJointOffset=<2 0 0>
     }]
   }
   n7 = Anim Dynamics - Bone: dyn_headfur_3_02!(SimulationSpace:AnimPhysSimSpaceType Enum, ComponentPose=@n6, Alpha=1, ExternalForce="0,0,0") -> (Pose) {
    angularDampingOverride=1
    angularSpringConstant=14000
    bAngularSpring=true
    bOverrideAngularDamping=true
    bUseGravityOverride=true
    gravityScale=0.35
    simulationSpace=RootRelative
    boundBone {boneName=dyn_headfur_3_02}
    componentAppliedLinearAccClamp=<2000 2000 2000>
    componentLinearAccScale=<1 1 1>
    componentLinearVelScale=<1 1 1>
    gravityOverride {z=-980}
    physicsBodyDefinitions [{
      boundBone {boneName=dyn_headfur_3_02}
      boxExtents=<25 30 30>
      constraintSetup {angularLimitsMax=<0 5 5>, angularLimitsMin=<0 -5 -5>, angularTarget=<1 0 0>}
      localJointOffset=<2 0 0>
     }]
   }
   n8 = "Transform (Modify) Bone - Bone: head"!(ComponentPose=@n7, bAlphaBoolEnabled:bool=True, Translation="0,0,0", Rotation="0,0,0", Scale="0.85,0.85,1") -> (Pose) {
    alphaInputType=Bool
    rotationSpace=BCS_BoneSpace
    scaleMode=BMM_Replace
    scaleSpace=BCS_BoneSpace
    translationSpace=BCS_BoneSpace
    boneToModify {boneName=head}
    scale=<0.85 0.85>
   }
   n9 = "Transform (Modify) Bone - Bone: neck_02"!(ComponentPose=@n8, bAlphaBoolEnabled:bool=True, Translation="0,0,0", Rotation="0,0,0", Scale="1,0.8,0.8") -> (Pose) {
    alphaInputType=Bool
    rotationSpace=BCS_BoneSpace
    scaleMode=BMM_Replace
    scaleSpace=BCS_BoneSpace
    translationSpace=BCS_BoneSpace
    boneToModify {boneName=neck_02}
    scale {y=0.8, z=0.8}
   }
   Output Pose(Result=Component To Local(ComponentPose=@n9))
  ]}
 EventGraph {nodes [
   Event BlueprintUpdateAnimation()
   > n1 = Sequence!(execute) -> (then_0, then_1, then_2:Exec)
   > n2 = Branch(Condition=Get bIsInFrontEnd())
   > Set Sim Space(Sim Space:AnimPhysSimSpaceType Enum=World) -> (Output_Get:AnimPhysSimSpaceType Enum)
   > Set DynAlpha(DynAlpha:double=0.3) -> (Output_Get:double)
   Set Sim Space(execute=@n2.else)
   > Set DynAlpha(DynAlpha=1)
   Set FortPawn(execute=@n1.then_1, FortPawn:Fort Pawn Object Reference=TryGetFortPawn()) -> (Output_Get:Fort Pawn Object Reference)
   n3 = Get FortPawn() -> (FortPawn:Fort Pawn Object Reference)
   > n4 = Is Valid(InputObject=@n3)
   Set weapon_visible(execute=@n4.Is Not Valid, weapon_visible:bool=false) -> (Output_Get:bool)
   n5 = Get CurrentWeapon!(self:Fort Pawn Object Reference=@n3) -> (CurrentWeapon:Fort Weapon Object Reference)
   n6 = GetCoreAnimation!(self:Fort Weapon Object Reference=@n5) -> (ReturnValue:EFortWeaponCoreAnimation Enum)
   n7 = Set notCoverCrosshair(
    execute=@n4.Is Valid
    notCoverCrosshair:bool=OR Boolean!(
     A="Equal (Enum)"(A:EFortWeaponCoreAnimation Enum=@n6, B:EFortWeaponCoreAnimation Enum=TrapPlacement)
     B="Equal (Enum)"(A=@n6, B=RocketLauncher)
     C:bool="Equal (Enum)"(A=@n6, B=GoingCommando)
    ) -> (ReturnValue)
   ) -> (Output_Get:bool)
   > n8 = Branch(Condition=@n7)
   > Set weapon_visible(weapon_visible=false)
   n9 = IsWeaponHolstered!(self:Fort Pawn Object Reference=@n3) -> (ReturnValue:bool)
   Set weapon_visible(execute=@n8.else, weapon_visible=AND Boolean(A=IsValid(Object=@n5), B=NOT Boolean(A=@n9)))
   n10 = Set isLocalPlayer(execute=@n1.then_2, isLocalPlayer:bool=IsLocalPawn!(Actor:Actor Object Reference=GetOwningActor()) -> (IsLocalPawn:bool)) -> (Output_Get:bool)
   > n11 = Branch(Condition=AND Boolean!(A=@n10, B=Get bIsTargeting(), C:bool=Get weapon_visible() -> (weapon_visible:bool)) -> (ReturnValue))
   > Set scaleDownAlpha(scaleDownAlpha:bool=true) -> (Output_Get:bool)
   Set scaleDownAlpha(execute=@n11.else, scaleDownAlpha=false)
   TryGetPawnOwner()
  ]}
]
variables [
 DynAlpha {type=number}
 FortPawn {type=FortPawn, default=null}
 isLocalPlayer {type=boolean}
 notCoverCrosshair {type=boolean}
 scaleDownAlpha {type=boolean}
 Sim Space {type=AnimPhysSimSpaceType}
 weapon_visible {type=boolean}
]
